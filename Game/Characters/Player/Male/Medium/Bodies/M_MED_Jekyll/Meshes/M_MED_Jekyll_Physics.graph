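# Physics asset for the medium male Jekyll player body, covering the tie, belt and inner and outer cuffs on both sides.
# Capsules on three upper spine bones form the torso, anchoring tie and cuffs at spine_05 and belt at spine_03.
# Tie and belt chains widen their swing limits toward the tip, while cuffs allow only slight movement.
# Spine links and the belt and cuff parent anchors have free angular motion.
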PhysicsAsset /Game/Characters/Player/Male/Medium/Bodies/M_MED_Jekyll/Meshes/M_MED_Jekyll_Physics
bodies {
 dyn_belt_01 [{shapeType=Capsule, length=1, radius=3.75, center=<0.05 0 0>, rotation=<rot 90 0 0>}]
 dyn_belt_02 [{shapeType=Capsule, length=3.75, radius=3.75, center=<1 0 0>, rotation=<rot 90 0 0>}]
 dyn_cuff_inner_01_l [{shapeType=Capsule, length=2.5, radius=1.75, rotation=<rot 90 0 0>}]
 dyn_cuff_inner_01_r [{shapeType=Capsule, length=5.75, radius=2.75, center=<0.01 0 0>, rotation=<rot 90 0 0>}]
 dyn_cuff_outer_01_l [{shapeType=Capsule, length=5.75, radius=2.75, center=<0.01 0 0>, rotation=<rot 90 0 0>}]
 dyn_cuff_outer_01_r [{shapeType=Capsule, length=2.5, radius=1.75, rotation=<rot 90 0 0>}]
 dyn_parent_belt_01 [{shapeType=Capsule, length=0.505, radius=0.505, rotation=<rot 0 -90 0>}]
 dyn_parent_cuff_inner_01_l [{shapeType=Capsule, length=0.505, radius=0.505, rotation=<rot 0 -90 0>}]
 dyn_parent_cuff_inner_01_r [{shapeType=Capsule, length=0.505, radius=0.505, rotation=<rot 0 -90 0>}]
 dyn_parent_cuff_outer_01_l [{shapeType=Capsule, length=0.505, radius=0.505, rotation=<rot 0 -90 0>}]
 dyn_parent_cuff_outer_01_r [{shapeType=Capsule, length=0.505, radius=0.505, rotation=<rot 0 -90 0>}]
 dyn_parent_tie [{shapeType=Capsule, length=2.5, radius=2, center=<2.0018 0.0001 1.3>, rotation=<rot 90 0 0>}]
 dyn_tie_01 [{shapeType=Capsule, length=2.25, radius=2.25, center=<1.5 0 1.5>, rotation=<rot 88.5217 0.0001 0.0001>}]
 dyn_tie_02 [{shapeType=Capsule, length=2, radius=3.75, center=<2 0.0001 1.9>, rotation=<rot 82.3029 0 0>}]
 dyn_tie_03 [{shapeType=Capsule, length=2.75, radius=4.5, center=<2.668 0 1.3>, rotation=<rot 82.5379 0 0>}]
 spine_03 [{shapeType=Capsule, length=10.75, radius=8.5}]
 spine_04 [{shapeType=Capsule, length=6.7434, radius=12.1321, center=<4.0404 1.2368 0.2457>, rotation=<rot 0 -90 0>}]
 spine_05 [{shapeType=Capsule, length=11.75, radius=11}]
}
constraints [
 {
  bone1Name=spine_05
  bone2Name=spine_04
  swing1LimitDegrees=45
  swing1Motion=Free
  swing2LimitDegrees=45
  swing2Motion=Free
  twistLimitDegrees=45
  twistMotion=Free
 }
 {bone1Name=dyn_parent_tie, bone2Name=spine_05, swing1LimitDegrees=5, swing2LimitDegrees=5, twistLimitDegrees=5}
 {bone1Name=dyn_tie_01, bone2Name=dyn_parent_tie, swing1LimitDegrees=10, swing2LimitDegrees=10, twistLimitDegrees=5}
 {bone1Name=dyn_tie_02, bone2Name=dyn_tie_01, swing1LimitDegrees=15, swing2LimitDegrees=15, twistLimitDegrees=5}
 {bone1Name=dyn_tie_03, bone2Name=dyn_tie_02, swing1LimitDegrees=20, swing2LimitDegrees=20, twistLimitDegrees=5}
 {
  bone1Name=spine_04
  bone2Name=spine_03
  swing1LimitDegrees=45
  swing1Motion=Free
  swing2LimitDegrees=45
  swing2Motion=Free
  twistLimitDegrees=45
  twistMotion=Free
 }
 {
  bone1Name=dyn_parent_belt_01
  bone2Name=spine_03
  swing1LimitDegrees=45
  swing1Motion=Free
  swing2LimitDegrees=45
  swing2Motion=Free
  twistLimitDegrees=45
  twistMotion=Free
 }
 {bone1Name=dyn_belt_01, bone2Name=dyn_parent_belt_01, swing1LimitDegrees=2, swing2LimitDegrees=2, twistLimitDegrees=5}
 {bone1Name=dyn_belt_02, bone2Name=dyn_belt_01, swing1LimitDegrees=15, swing2LimitDegrees=15, twistLimitDegrees=5}
 {
  bone1Name=dyn_parent_cuff_inner_01_l
  bone2Name=spine_05
  swing1LimitDegrees=45
  swing1Motion=Free
  swing2LimitDegrees=45
  swing2Motion=Free
  twistLimitDegrees=45
  twistMotion=Free
 }
 {
  bone1Name=dyn_parent_cuff_outer_01_l
  bone2Name=spine_05
  swing1LimitDegrees=45
  swing1Motion=Free
  swing2LimitDegrees=45
  swing2Motion=Free
  twistLimitDegrees=45
  twistMotion=Free
 }
 {
  bone1Name=dyn_parent_cuff_outer_01_r
  bone2Name=spine_05
  swing1LimitDegrees=45
  swing1Motion=Free
  swing2LimitDegrees=45
  swing2Motion=Free
  twistLimitDegrees=45
  twistMotion=Free
 }
 {
  bone1Name=dyn_parent_cuff_inner_01_r
  bone2Name=spine_05
  swing1LimitDegrees=45
  swing1Motion=Free
  swing2LimitDegrees=45
  swing2Motion=Free
  twistLimitDegrees=45
  twistMotion=Free
 }
 {bone1Name=dyn_cuff_inner_01_l, bone2Name=dyn_parent_cuff_inner_01_l, swing1LimitDegrees=5, swing2LimitDegrees=5, twistLimitDegrees=1}
 {bone1Name=dyn_cuff_outer_01_l, bone2Name=dyn_parent_cuff_outer_01_l, swing1LimitDegrees=5, swing2LimitDegrees=5, twistLimitDegrees=1}
 {bone1Name=dyn_cuff_inner_01_r, bone2Name=dyn_parent_cuff_inner_01_r, swing1LimitDegrees=5, swing2LimitDegrees=5, twistLimitDegrees=1}
 {bone1Name=dyn_cuff_outer_01_r, bone2Name=dyn_parent_cuff_outer_01_r, swing1LimitDegrees=5, swing2LimitDegrees=5, twistLimitDegrees=1}
]
