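# Post-process style AnimBlueprint for the M_MED_Jekyll cosmetic body, layering secondary bone fixes and physics over the pose copied from the attached parent mesh.
# Weighted constraints drive the cuff parent bones, additive offsets shift tie and backpack bones, and a RigidBody node simulates last.
# The offset alpha follows JH_TransformCurve above a threshold during the ManAndMonster ToMonster emote.
# Customization load toggles a material parameter hiding an element when backbling is worn.
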AnimBlueprint /BRCosmetics/Characters/Player/Male/Medium/Bodies/M_MED_Jekyll/Meshes/M_MED_Jekyll_AnimBP
target_skeleton=/Game/Characters/Player/Male/Medium/Bodies/M_MED_Jekyll/Meshes/M_MED_Jekyll_Skeleton
graphs [
 AnimGraph {nodes [
   n1 = Get tie_alpha() -> (tie_alpha:double)
   "Transform (Modify) Bone - Bone: attach_cape"(Alpha=@n1, Translation="0,-0.75,1", Rotation="0,0,0", Scale="1,1,1") {
    rotationMode=BMM_Additive
    rotationSpace=BCS_BoneSpace
    translationMode=BMM_Additive
    translationSpace=BCS_BoneSpace
    alphaBoolBlend {blendOption=Linear}
    alphaScaleBiasClamp {bInterpResult=true, interpSpeedDecreasing=5, interpSpeedIncreasing=25}
    boneToModify {boneName=attach_cape}
    translation {y=-0.75, z=1}
   }
   n2 = "Constraint - dyn_parent_cuff_inner_01_l (PPPPPP)"!(
    ComponentPose=Local To Component(LocalPose=Copy Pose From Mesh() {bCopyCurves=true, bUseAttachedParent=true})
    Alpha=1
    ConstraintWeights_0=0.0704
    ConstraintWeights_1:float=0.2694
    ConstraintWeights_2:float=0.1006
    ConstraintWeights_3:float=0.049
    ConstraintWeights_4:float=0.4873
    ConstraintWeights_5:float=0.0234
   ) -> (Pose) {
    alphaBoolBlend {blendOption=Linear}
    boneToModify {boneName=dyn_parent_cuff_inner_01_l}
    constraintSetup [
     {targetBone {boneName=dyn_cuff_inner_01_l}}
     {targetBone {boneName=upperarm_twist_02_l}}
     {targetBone {boneName=upperarm_twist_01_l}}
     {targetBone {boneName=lowerarm_l}}
     {targetBone {boneName=upperarm_l}}
     {targetBone {boneName=lowerarm_twist_02_l}}
    ]
    constraintWeights [1,1,1,1,1,1]
   }
   n3 = "Constraint - dyn_parent_cuff_outer_01_l (PPPPPP)"!(
    ComponentPose=@n2
    Alpha=1
    ConstraintWeights_0=0.0991
    ConstraintWeights_1:float=0.4821
    ConstraintWeights_2:float=0.0836
    ConstraintWeights_3:float=0.0145
    ConstraintWeights_4:float=0.3061
    ConstraintWeights_5:float=0.0147
   ) -> (Pose) {
    alphaBoolBlend {blendOption=Linear}
    boneToModify {boneName=dyn_parent_cuff_outer_01_l}
    constraintSetup [
     {targetBone {boneName=dyn_cuff_outer_01_l}}
     {targetBone {boneName=upperarm_twist_02_l}}
     {targetBone {boneName=upperarm_twist_01_l}}
     {targetBone {boneName=lowerarm_l}}
     {targetBone {boneName=upperarm_l}}
     {targetBone {boneName=lowerarm_twist_02_l}}
    ]
    constraintWeights [1,1,1,1,1,1]
   }
   n4 = "Constraint - dyn_parent_cuff_inner_01_r (PPPPPP)"!(
    ComponentPose=@n3
    Alpha=1
    ConstraintWeights_0=0.0219
    ConstraintWeights_1:float=0.4651
    ConstraintWeights_2:float=0.0456
    ConstraintWeights_3:float=0.0987
    ConstraintWeights_4:float=0.2729
    ConstraintWeights_5:float=0.0957
   ) -> (Pose) {
    alphaBoolBlend {blendOption=Linear}
    boneToModify {boneName=dyn_parent_cuff_inner_01_r}
    constraintSetup [
     {targetBone {boneName=lowerarm_twist_02_r}}
     {targetBone {boneName=upperarm_r}}
     {targetBone {boneName=lowerarm_r}}
     {targetBone {boneName=upperarm_twist_01_r}}
     {targetBone {boneName=upperarm_twist_02_r}}
     {targetBone {boneName=dyn_cuff_inner_01_r}}
    ]
    constraintWeights [1,1,1,1,1,1]
   }
   n5 = "Constraint - dyn_parent_cuff_outer_01_r (PPPPPP)"!(
    ComponentPose=@n4
    Alpha=1
    ConstraintWeights_0=0.0147
    ConstraintWeights_1:float=0.3061
    ConstraintWeights_2:float=0.0145
    ConstraintWeights_3:float=0.0836
    ConstraintWeights_4:float=0.4821
    ConstraintWeights_5:float=0.0991
   ) -> (Pose) {
    alphaBoolBlend {blendOption=Linear}
    boneToModify {boneName=dyn_parent_cuff_outer_01_r}
    constraintSetup [
     {targetBone {boneName=lowerarm_twist_02_r}}
     {targetBone {boneName=upperarm_r}}
     {targetBone {boneName=lowerarm_r}}
     {targetBone {boneName=upperarm_twist_01_r}}
     {targetBone {boneName=upperarm_twist_02_r}}
     {targetBone {boneName=dyn_cuff_outer_01_r}}
    ]
    constraintWeights [1,1,1,1,1,1]
   }
   n6 = "Transform (Modify) Bone - Bone: dyn_parent_tie"(ComponentPose=@n5, Alpha=Get tie_alpha(), Translation="-4,0.000,6", Rotation="2.088,0.000,0.000", Scale="1,1,1") {
    rotationMode=BMM_Additive
    rotationSpace=BCS_BoneSpace
    translationMode=BMM_Additive
    translationSpace=BCS_BoneSpace
    alphaBoolBlend {blendOption=Linear}
    alphaScaleBiasClamp {bInterpResult=true, interpSpeedDecreasing=5, interpSpeedIncreasing=25}
    boneToModify {boneName=dyn_parent_tie}
    rotation {pitch=2.088}
    translation {x=-4, z=6}
   }
   n7 = "Transform (Modify) Bone - Bone: attach_backpack"(ComponentPose=@n6, Alpha=@n1, Translation="0,-0.75,1", Rotation="0,0,0", Scale="1,1,1") {
    rotationMode=BMM_Additive
    rotationSpace=BCS_BoneSpace
    translationMode=BMM_Additive
    translationSpace=BCS_BoneSpace
    alphaBoolBlend {blendOption=Linear}
    alphaScaleBiasClamp {bInterpResult=true, interpSpeedDecreasing=5, interpSpeedIncreasing=25}
    boneToModify {boneName=attach_backpack}
    translation {y=-0.75, z=1}
   }
   n8 = RigidBody(ComponentPose=@n7) {
    bOverrideWorldGravity=true
    overridePhysicsAsset=/Game/Characters/Player/Male/Medium/Bodies/M_MED_Jekyll/Meshes/M_MED_Jekyll_Physics
    simulationSpace=BaseBoneSpace
    alphaBoolBlend {blendOption=Linear}
    baseBoneRef {boneName=root}
    componentAppliedLinearAccClamp=<2000 2000 1500>
    componentLinearAccScale=<0.5 0.5 0.8>
    componentLinearVelScale=<0.5 0.5 0.8>
    overrideWorldGravity {z=-980}
   }
   Output Pose(Result=Component To Local(ComponentPose=@n8)) {layerGroup=None}
  ]}
 EventGraph {nodes [
   Event BlueprintUpdateAnimation()
   > Get!(execute:Exec) -> (then:Exec, else:Exec, AnimInst:Anim Instance Object Reference)
   n1 = Get AnimInst() -> (AnimInst:Anim Instance Object Reference)
   > n2 = Branch(Condition=Montage_IsActive!(
     self:Anim Instance Object Reference=@n1
     Montage:Anim Montage Object Reference=/BRCosmetics/Animation/Game/MainPlayer/Emotes/ManAndMonster/Emote_ManAndMonster_ToMonster_CMM_M
    ) -> (ReturnValue:bool))
   n3 = GetCurveValue(self=@n1, CurveName=JH_TransformCurve)
   > n4 = Branch(Condition="float >= float"(A=@n3, B=0.6) {is=title K2Node_CallFunction})
   > Set tie_alpha(tie_alpha:double=MapRangeClamped(Value=@n3, InRangeA=0.3, InRangeB=1)) -> (Output_Get:double)
   Set tie_alpha(execute=[@n2.else, @n4.else], tie_alpha=0)
   Event OnCharacterCustomizationLoaded()
   n5 = CheckForBackbling()
   > n6 = Set is_wearing_backbling(is_wearing_backbling:bool=OR Boolean(A=@n5, B=@n5.bIsBackblingLong)) -> (Output_Get:bool)
   > n7 = Branch(Condition=@n6)
   > n8 = Get Dynamic Material Instance(SkeletalMeshComponent=GetPartSkeletalMeshComponent())
   > SetScalarParameterValue(self=@n8, ParameterName=Hide Element 01, Value=1)
   n9 = Get Dynamic Material Instance(execute=@n7.else)
   > SetScalarParameterValue(self=@n9, ParameterName=Hide Element 01, Value=0)
   Event BlueprintInitializeAnimation()
   n10 = GetOwnerSkeletalMeshComponent()
   > Is Valid(InputObject=@n10)
   > Set Skeletal_Mesh(Skeletal_Mesh:Skeletal Mesh Component Object Reference=@n10) -> (Output_Get:Skeletal Mesh Component Object Reference)
   > n11 = Cast To FortPlayerPawn(Object=GetOwningActor())
   > Is Valid(InputObject=@n11)
   > n12 = Set Pawn(Pawn:Fort Player Pawn Object Reference=@n11) -> (Output_Get:Fort Player Pawn Object Reference)
   n13 = GetAnimInstance(self=Get Mesh(self=@n12))
   > Is Valid(InputObject=@n13)
   > Set AnimInst(AnimInst:Anim Instance Object Reference=@n13) -> (Output_Get:Anim Instance Object Reference)
   TryGetPawnOwner()
  ]}
]
variables [
 AnimInst {type=AnimInstance, default=null}
 is_wearing_backbling {type=boolean}
 Pawn {type=FortPlayerPawn}
 Skeletal_Mesh {type=SkeletalMeshComponent}
 tie_alpha {type=number}
]
